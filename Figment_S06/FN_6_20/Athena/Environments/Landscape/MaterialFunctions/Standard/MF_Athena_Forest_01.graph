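# Landscape material function for the Athena forest floor layer, returning material attributes plus a separate Height output.
# It samples the Forest texture parameter at input UVs, defaulting to texture coordinates, using a shared world-group wrap sampler.
# RGB feeds BaseColor, while alpha serves as height, Opacity and the blend between the Forest_SpecMin and Forest_SpecMax Specular parameters.
# Roughness and Normal are unconnected; a flat-normal constant feeds only an unused reroute.
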MaterialFunction /Figment_S06/FN_6_20/Athena/Environments/Landscape/MaterialFunctions/Standard/MF_Athena_Forest_01
graph {nodes [
  n1 = TextureSampleParameter2D(UVs=FunctionInput(Preview=TextureCoordinate()) {bUsePreviewValueAsDefault=true, inputName=UVs, inputType=FunctionInput_Vector2}) {
   group=Forest
   parameterName=Forest
   samplerSource=SSM_Wrap_WorldGroupSettings
   texture=/Figment_S06/FN_6_20/Athena/Environments/Landscape/Textures/T_Athena_ForestFloor_D
  }
  n2 = ScalarParameter() {group=Forest, parameterName=Forest_SpecMin}
  n3 = ScalarParameter() {defaultValue=1, group=Forest, parameterName=Forest_SpecMax}
  FunctionOutput(None=MakeMaterialAttributes!(
    BaseColor=@n1.RGB
    Metallic
    Specular=LinearInterpolate(A=@n2, B=@n3, Alpha=@n1.A)
    Roughness
    Anisotropy
    EmissiveColor
    Opacity=@n1.A
    OpacityMask
    Normal
    Tangent
    WorldPositionOffset
    SubsurfaceColor
    ClearCoat
    ClearCoatRoughness
    AmbientOcclusion
    Refraction
    CustomizedUVs_0
    CustomizedUVs_1
    CustomizedUVs_2
    CustomizedUVs_3
    CustomizedUVs_4
    CustomizedUVs_5
    CustomizedUVs_6
    CustomizedUVs_7
    PixelDepthOffset
    ShadingModel
    Displacement
   ))
  FunctionOutput(None=@n1.A) {outputName=Height, sortPriority=1}
  Reroute(None=Constant3Vector() {constant=<rgba 0 0 1 0>})
 ]}
