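# Landscape material function that turns slope angle into blend alphas for Athena environments.
# MF_Athena_SlopeAngle feeds HeightLerp as the height texture, with the Height input as transition phase and a Grass-group Contrast parameter, producing Alpha and Alpha_NoContrast.
# A second HeightLerp adds ColormapTransitionBias from MPC_Landscape to Height, producing a separately shifted Alpha_ColorMap for the colormap texture.
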MaterialFunction /Figment_S06/FN_6_20/Athena/Environments/Landscape/MaterialFunctions/Utilities/MF_Athena_SlopeBlend
graph {nodes [
  n1 = FunctionInput() {inputName=Height, inputType=FunctionInput_Scalar}
  n2 = MaterialFunctionCall() {materialFunction=&/MF_Athena_SlopeAngle}
  n3 = ScalarParameter() {defaultValue=5, group=Grass, parameterName=Contrast}
  n4 = MaterialFunctionCall(Transition Phase=@n1, Height Texture=@n2.Result, Contrast=@n3) {materialFunction=&HeightLerp}
  FunctionOutput(None=@n4.Alpha) {outputName=Alpha}
  FunctionOutput(None=@n4.Lerp Alpha No Contrast) {outputName=Alpha_NoContrast, sortPriority=1}
  n5 = MaterialFunctionCall(
   Transition Phase=Add(
    A=CollectionParameter() {collection=/FigmentEnvironment/FN_6_20/Athena/Environments/Landscape/MPC/MPC_Landscape, parameterName=ColormapTransitionBias}
    B=@n1
   )
   Height Texture=@n2.Result
   Contrast=@n3
  ) {
   desc=Height lerp for ColorMap Texture - has ColormapTransitionBias
   materialFunction=&HeightLerp
  }
  FunctionOutput(None=@n5.Alpha) {outputName=Alpha_ColorMap, sortPriority=2}
 ]}
