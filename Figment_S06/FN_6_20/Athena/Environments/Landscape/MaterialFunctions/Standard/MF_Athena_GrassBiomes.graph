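# Landscape material function that builds grass biome variants (grass only, grass topology, farm, mountain) from Rock and Grass material attribute inputs.
# Rock is brightness-graded by world height, then layered with grass using a slope blend whose height input multiplies both materials' opacity.
# Each biome applies its own color blend, with shared rock tint from a parameter collection and NoTopo layer-driven grass tint.
# Colors and blend alphas are also output separately.
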MaterialFunction /Figment_S06/FN_6_20/Athena/Environments/Landscape/MaterialFunctions/Standard/MF_Athena_GrassBiomes
graph {nodes [
  n1 = FunctionInput() {inputName=Rock, inputType=FunctionInput_MaterialAttributes}
  n2 = CollectionParameter() {collection=/FigmentEnvironment/FN_6_20/Athena/Environments/Landscape/MPC/MPC_Landscape, parameterName=Rock_Bottom_Brightness}
  n3 = CollectionParameter() {collection=&MPC_Landscape, parameterName=Rock_Top_Brightness}
  n4 = MaterialFunctionCall!(height_Offset, height_GradScale, Coords=WorldPosition().XYZ) {materialFunction=/FigmentEnvironment/FN_6_20/Athena/Environments/Landscape/MaterialFunctions/MF_Athena_WorldHeightGrad}
  n5 = FunctionInput() {inputName=Grass, inputType=FunctionInput_MaterialAttributes}
  n6 = MaterialFunctionCall!(Mat attributes=@n5) {materialFunction=/Engine/Functions/Engine_MaterialFunctions02/MatLayerBlend_BreakOpacity}
  n7 = MaterialFunctionCall(Mat attributes=@n1) {materialFunction=&MatLayerBlend_BreakOpacity}
  n8 = MaterialFunctionCall!(Height=Multiply(A=Multiply(A=@n6.Result, B=@n7.Result), B=ScalarParameter() {defaultValue=0.08, group=Grass, parameterName=Transition_Phase})) {materialFunction=/Figment_S06/FN_6_20/Athena/Environments/Landscape/MaterialFunctions/Utilities/MF_Athena_SlopeBlend}
  n9 = MaterialFunctionCall!(
   Base Material=MaterialFunctionCall!(In=@n1, Color=LinearInterpolate(A=@n2, B=@n3, Alpha=@n4.Result)) {materialFunction=/FigmentEnvironment/FN_6_20/Athena/Environments/Landscape/MaterialFunctions/MF_Athena_Material_BaseMultiply}.Result
   Top Material=@n5
   Alpha=@n8.Alpha
  ) {materialFunction=/Engine/Functions/MaterialLayerFunctions/MatLayerBlend_Standard}
  n10 = ComponentMask(None=CollectionParameter() {collection=&MPC_Landscape, parameterName=RockTint}) {b=true, g=true, r=true}
  n11 = LinearInterpolate(A=ScalarParameter() {defaultValue=0.8, group=Grass, parameterName=Grass_Tint}, Alpha=LandscapeLayerSample() {parameterName=NoTopo})
  FunctionOutput(None=MaterialFunctionCall(In=@n9.Blended Material, Color=LinearInterpolate(A=@n10, B=@n11, Alpha=@n8.Alpha)) {materialFunction=&MF_Athena_Material_BaseMultiply}.Result) {outputName=Grass_Topology, sortPriority=4}
  n12 = MaterialFunctionCall!(Rock_Diffuse, Alpha, ColorBlend_Alpha=@n8.Alpha, ColorBlend_Alpha_Bias=@n8.Alpha, In, Grass_Diffuse) {materialFunction=/FigmentEnvironment/FN_6_20/Athena/Environments/Landscape/MaterialFunctions/MF_FarmGrass_Colors}
  FunctionOutput(None=MaterialFunctionCall(In=@n9.Blended Material, Color=@n12.BlendedColors) {materialFunction=&MF_Athena_Material_BaseMultiply}.Result) {outputName=Farm, sortPriority=6}
  FunctionOutput(None=@n12.BlendedColors_Bias) {outputName=Farm_Colors, sortPriority=11}
  FunctionOutput() {outputName=----------------, sortPriority=7}
  FunctionOutput(None=@n11) {outputName=Grass_Only_Colors, sortPriority=8}
  FunctionOutput(None=@n8.Alpha) {outputName=BlendAlpha, sortPriority=13}
  FunctionOutput(None=@n8.Alpha_NoContrast) {outputName=BlendAlpha_NoContrast, sortPriority=14}
  FunctionOutput() {outputName=-----------------, sortPriority=12}
  FunctionOutput(None=LinearInterpolate(A=@n10, B=@n11, Alpha=@n8.Alpha_ColorMap) {desc=Colors Only}) {outputName=Grass_Topology_Colors, sortPriority=9}
  FunctionOutput(None=MaterialFunctionCall(In=@n5, Color=@n11) {materialFunction=&MF_Athena_Material_BaseMultiply}.Result) {outputName=Grass_Only, sortPriority=3}
  n13 = MaterialFunctionCall!(Rock_Diffuse, ColorBlend_Alpha=@n8.Alpha, ColorBlend_Alpha_Bias=@n8.Alpha, Grass_Diffuse) {materialFunction=/FigmentEnvironment/FN_6_20/Athena/Environments/Landscape/MaterialFunctions/MF_MountainGrass_Colors}
  FunctionOutput(None=MaterialFunctionCall(In=@n9.Blended Material, Color=@n13.Blended_Colors) {materialFunction=&MF_Athena_Material_BaseMultiply}.Result) {outputName=Mountain, sortPriority=5}
  FunctionOutput(None=@n13.Blended_Colors_Bias) {outputName=Mountain_Colors, sortPriority=10}
 ]}
